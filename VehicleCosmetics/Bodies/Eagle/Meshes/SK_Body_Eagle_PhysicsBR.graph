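# Physics asset for the Eagle vehicle body skeletal mesh in VehicleCosmetics, named as the PhysicsBR variant.
# It holds a single body on the root bone, with no collision shapes listed.
# It defines no constraints, so it sets up no jointed or ragdoll simulation between bones.
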PhysicsAsset /VehicleCosmetics/Bodies/Eagle/Meshes/SK_Body_Eagle_PhysicsBR
bodies {root []}
constraints []
